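# Post-process style Anim Blueprint for the Progressive Jonesy DMG 1 body that adds secondary motion to its crossbow and shirt.
# It copies the parent mesh pose, then adjusts crossbow and shirt master bones with constraints and an additive offset.
# Two chained RigidBody nodes simulate crossbow and shirt physics assets in base-bone space.
# Their gravity override is windVector, updated each frame to scaled skydiving force while skydiving, otherwise normal gravity.
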AnimBlueprint /BRCosmetics/Characters/Player/Male/Medium/Bodies/M_MED_Progressive_Jonesy_DMG_1/Meshes/M_MED_Progressive_Jonesy_DMG_1_AnimBP
target_skeleton=&/M_MED_Progressive_Jonesy_DMG_1_Skeleton
graphs [
 AnimGraph {nodes [
   n1 = "Constraint - dyn_crossbow_01_l_MSTR (PP)"!(
    ComponentPose=Local To Component(LocalPose=Copy Pose From Mesh() {bUseAttachedParent=true})
    Alpha=1
    ConstraintWeights_0=0.5
    ConstraintWeights_1:float=0.5
   ) -> (Pose) {
    alphaBoolBlend {blendOption=Linear}
    boneToModify {boneName=dyn_crossbow_01_l_MSTR}
    constraintSetup [{targetBone {boneName=pelvis}}, {targetBone {boneName=thigh_l}}]
    constraintWeights [1,1]
   }
   n2 = "Constraint - dyn_shirt_fr_01_r_MSTR (R)"(ComponentPose=@n1, Alpha=1, ConstraintWeights_0=1) {
    alphaBoolBlend {blendOption=Linear}
    boneToModify {boneName=dyn_shirt_fr_01_r_MSTR}
    constraintSetup [{transformType=Rotation, targetBone {boneName=thigh_twist_01_r}}]
    constraintWeights [1]
   }
   n3 = "Constraint - dyn_shirt_01_r_MSTR (R)"(ComponentPose=@n2, Alpha=0, ConstraintWeights_0=1) {
    alphaBoolBlend {blendOption=Linear}
    boneToModify {boneName=dyn_shirt_01_r_MSTR}
    constraintSetup [{transformType=Rotation, targetBone {boneName=thigh_r}}]
    constraintWeights [1]
   }
   n4 = "Transform (Modify) Bone - Bone: dyn_shirt_fr_01_r_MSTR"(ComponentPose=@n3, Alpha=1, Translation="-0.000,0,0.371", Rotation="0,0,0", Scale="1,1,1") {
    rotationMode=BMM_Additive
    rotationSpace=BCS_BoneSpace
    scaleMode=BMM_Additive
    scaleSpace=BCS_BoneSpace
    translationMode=BMM_Additive
    translationSpace=BCS_BoneSpace
    alphaBoolBlend {blendOption=Linear}
    boneToModify {boneName=dyn_shirt_fr_01_r_MSTR}
    translation {z=0.371}
   }
   n5 = Get windVector() -> (windVector:Vector)
   n6 = RigidBody!(ComponentPose=@n4, Alpha, OverrideWorldGravity:Vector=@n5, ExternalForce) -> (Pose) {
    bClampLinearTranslationLimitToRefPose=true
    bOverrideWorldGravity=true
    overridePhysicsAsset=&/M_MED_Progressive_Jonesy_DMG_1_Crossbow_Physics
    simulationSpace=BaseBoneSpace
    alphaBoolBlend {blendOption=Linear}
    baseBoneRef {boneName=root}
    componentAppliedLinearAccClamp=<3000 3000 3000>
    componentLinearAccScale=<0.3 0.3 0.7>
    componentLinearVelScale=<0.3 0.3 0.7>
    overrideWorldGravity {z=-980}
   }
   n7 = RigidBody(ComponentPose=@n6, OverrideWorldGravity=@n5) {
    bClampLinearTranslationLimitToRefPose=true
    bOverrideWorldGravity=true
    overridePhysicsAsset=&/M_MED_Progressive_Jonesy_DMG_1_Shirt_Physics
    simulationSpace=BaseBoneSpace
    alphaBoolBlend {blendOption=Linear}
    baseBoneRef {boneName=spine_05}
    componentAppliedLinearAccClamp=<3000 3000 3000>
    componentLinearAccScale=<0.3 0.3 0.4>
    componentLinearVelScale=<0.3 0.3 0.4>
    overrideWorldGravity {z=-980}
   }
   Output Pose(Result=Component To Local(ComponentPose=@n7)) {layerGroup=None}
  ]}
 EventGraph {nodes [
   Event BlueprintUpdateAnimation()
   n1 = Break Fort Anim Input Skydiving External Force(FortAnimInput_SkydivingExternalForce=Get SkydivingExternalForceInput()) -> (
    bUseSkydivingVectorForce=~
    SkydivingVectorForce=~
    HeadToPelvisDirection=~
    FloatingMultiplier=~
    FloatingAdditive=~
    DivingMultiplier=~
    DivingAdditive=~
    ParachutingMultiplier=~
    ParachutingAdditive=~
   )
   > Set windVector(windVector:Vector=SelectVector(A=vector * float(A=@n1.SkydivingVectorForce, B=0.15), B="0, 0,-980", bPickA=Get bIsSkyDiving())) -> (Output_Get:Vector)
  ]}
]
variables [windVector {type=Vector}]
